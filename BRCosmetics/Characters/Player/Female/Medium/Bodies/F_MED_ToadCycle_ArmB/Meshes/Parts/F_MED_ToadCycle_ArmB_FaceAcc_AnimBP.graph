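# Post-style AnimBP for the Female Medium ToadCycle ArmB face accessory, animating scarf and ponytail secondary motion.
# It copies the parent mesh pose, offsets scarf bones, and tilts the ponytail base by head yaw relative to spine_05.
# A backbling check selects between two spine_05-space RigidBody simulations differing in acceleration scale, followed by ponytail bone offsets.
# Meta-tag checks set arm and scarf flags, which the AnimGraph never reads.
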AnimBlueprint /BRCosmetics/Characters/Player/Female/Medium/Bodies/F_MED_ToadCycle_ArmB/Meshes/Parts/F_MED_ToadCycle_ArmB_FaceAcc_AnimBP
target_skeleton=&/F_MED_ToadCycle_ArmB_FaceAcc_Skeleton
graphs [
 AnimGraph {nodes [
   n1 = RigidBody(ComponentPose=Local To Component(LocalPose=Use cached pose 'hair pose'())) {
    bClampLinearTranslationLimitToRefPose=true
    bOverrideWorldGravity=true
    overridePhysicsAsset=&/F_MED_ToadCycle_ArmB_FaceAcc_Physics
    simulationSpace=BaseBoneSpace
    alphaBoolBlend {blendOption=Linear}
    baseBoneRef {boneName=spine_05}
    componentLinearAccScale=<0.02 0.375 0.375>
    componentLinearVelScale=<0.375 0.375 0.375>
    overrideWorldGravity {z=-980}
   }
   n2 = Component To Local(ComponentPose=@n1)
   n3 = RigidBody(ComponentPose=Local To Component(LocalPose=Use cached pose 'hair pose'())) {
    bClampLinearTranslationLimitToRefPose=true
    bOverrideWorldGravity=true
    overridePhysicsAsset=&F_MED_ToadCycle_ArmB_FaceAcc_Physics
    simulationSpace=BaseBoneSpace
    alphaBoolBlend {blendOption=Linear}
    baseBoneRef {boneName=spine_05}
    componentLinearAccScale=<0.375 0.375 0.375>
    componentLinearVelScale=<0.375 0.375 0.375>
    overrideWorldGravity {z=-980}
   }
   n4 = Component To Local(ComponentPose=@n3)
   n5 = Get isWearingPackBool() -> (isWearingPackBool:bool)
   n6 = "Transform (Modify) Bone - Bone: per3_ponytail_b"!(
    ComponentPose=Local To Component(LocalPose=Blend Poses by bool(BlendPose_0=@n2, BlendPose_1=@n4, bActiveValue=@n5) {blendType=Linear})
    bAlphaBoolEnabled:bool=True
    Translation="-0.4,0,0"
    Rotation="0,-12,0"
    Scale="1,1,1"
   ) -> (Pose) {
    alphaInputType=Bool
    rotationMode=BMM_Additive
    rotationSpace=BCS_BoneSpace
    scaleMode=BMM_Additive
    scaleSpace=BCS_BoneSpace
    translationMode=BMM_Additive
    translationSpace=BCS_BoneSpace
    alphaBoolBlend {blendInTime=0.2, blendOutTime=0.2}
    boneToModify {boneName=per3_ponytail_b}
    rotation {yaw=-12}
    translation {x=-0.4}
   }
   n7 = "Transform (Modify) Bone - Bone: per3_ponytail_c"!(ComponentPose=@n6, bAlphaBoolEnabled:bool=True, Translation="-0.8,0,0", Rotation="0,-3,0", Scale="1,1,1") -> (Pose) {
    alphaInputType=Bool
    rotationMode=BMM_Additive
    rotationSpace=BCS_BoneSpace
    scaleMode=BMM_Additive
    scaleSpace=BCS_BoneSpace
    translationMode=BMM_Additive
    translationSpace=BCS_BoneSpace
    alphaBoolBlend {blendInTime=0.2, blendOutTime=0.2}
    boneToModify {boneName=per3_ponytail_c}
    rotation {yaw=-3}
    translation {x=-0.8}
   }
   Output Pose(Result=Component To Local(ComponentPose=@n7)) {layerGroup=None}
   n8 = "Transform (Modify) Bone - Bone: scarf"(
    ComponentPose=Local To Component(LocalPose=Copy Pose From Mesh() {bUseAttachedParent=true})
    Alpha=1
    Translation="1.2,0,0"
    Rotation="-20,0,0"
    Scale="1,1,1"
   ) {
    rotationMode=BMM_Additive
    rotationSpace=BCS_BoneSpace
    scaleMode=BMM_Additive
    scaleSpace=BCS_BoneSpace
    translationMode=BMM_Additive
    translationSpace=BCS_BoneSpace
    alphaBoolBlend {blendInTime=0.2, blendOutTime=0.2}
    boneToModify {boneName=scarf}
    rotation {pitch=-20}
    translation {x=1.2}
   }
   n9 = "Transform (Modify) Bone - Bone: neck_01_scarf"(ComponentPose=@n8, Alpha=1, Translation="0,-2,0", Rotation="0,0,0", Scale="1,1,1") {
    rotationMode=BMM_Additive
    rotationSpace=BCS_BoneSpace
    scaleMode=BMM_Additive
    scaleSpace=BCS_BoneSpace
    translationMode=BMM_Additive
    translationSpace=BCS_BoneSpace
    alphaBoolBlend {blendInTime=0.2, blendOutTime=0.2}
    boneToModify {boneName=neck_01_scarf}
    translation {y=-2}
   }
   n10 = "Transform (Modify) Bone - Bone: per3_ponytail_a_base"(ComponentPose=@n9, Alpha=Get headUp() -> (headUp:double), Translation="0,0,0", Rotation="0,-40,0", Scale="1,1,1") {
    rotationMode=BMM_Additive
    rotationSpace=BCS_BoneSpace
    translationMode=BMM_Additive
    translationSpace=BCS_BoneSpace
    alphaBoolBlend {blendOption=Linear}
    alphaScaleBiasClamp {bMapRange=true, inRange {max=0, min=-50}, outRange {max=0, min=1}}
    boneToModify {boneName=per3_ponytail_a_base}
    rotation {yaw=-40}
   }
   Save cached pose 'hair pose'(Pose=Component To Local(ComponentPose=@n10)) {cacheName=hair pose}
  ]}
 EventGraph {nodes [
   Event BlueprintUpdateAnimation()
   > n1 = Sequence!(execute) -> (then_0, then_1, then_2:Exec)
   n2 = GetDeltaTransformFromRefPose(self=GetOwnerSkeletalMeshComponent(), BoneName=head, BaseName=spine_05)
   > Set headUp(headUp:double=BreakRotator(InRot=BreakTransform(InTransform=@n2).Rotation).Yaw) -> (Output_Get:double)
   n3 = Branch(execute=@n1.then_1, Condition=CharacterItemDefinitionHasAnyMetaTag(Tags=Make Array("[0]"=(TagName="Cosmetics.MetaTags.ToadCycleArm"))))
   > Set On/Off_Arm(On/Off_Arm:bool=true) -> (Output_Get:bool)
   Set On/Off_Arm(execute=@n3.else, On/Off_Arm=false)
   n4 = Branch(execute=@n1.then_2, Condition=CharacterItemDefinitionHasAnyMetaTag(Tags=Make Array("[0]"=(TagName="Cosmetics.MetaTags.ToadCycleSarf"))))
   > Set On/Off_Scarf(On/Off_Scarf:double=1) -> (Output_Get:double)
   Set On/Off_Scarf(execute=@n4.else, On/Off_Scarf=0)
   Event OnCharacterCustomizationLoaded()
   n5 = CheckForBackbling()
   > Set isWearingPackBool(isWearingPackBool:bool=OR Boolean!(A=@n5, B=@n5.bIsBackblingACape, C:bool=@n5.bIsBackblingLong) -> (ReturnValue)) -> (Output_Get:bool)
   > Set is_WearingCape(is_WearingCape:bool=@n5.bIsBackblingLong) -> (Output_Get:bool)
   AND Boolean(A=@n5, B=NOT Boolean(A=@n5.bIsBackblingLong))
  ]}
]
variables [
 headUp {type=number}
 is_WearingCape {type=boolean}
 isWearingPackBool {type=boolean}
 On/Off_Arm {type=boolean}
 On/Off_Scarf {type=number}
]
